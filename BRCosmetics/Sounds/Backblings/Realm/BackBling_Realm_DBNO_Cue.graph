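# Sound cue for the Realm back bling's down-but-not-out (DBNO) reaction, layering a staged damage, hover, signal and voice sequence.
# A root mixer plays a damage-received hit immediately, then after a delay a nested mixer starts a searching hover, followed by a random signal and, later, a random vox line, each on fixed delays.
# It uses close-range attenuation, the backbling sound class and per-owner concurrency that blocks overlapping plays.
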SoundCue /BRCosmetics/Sounds/Backblings/Realm/BackBling_Realm_DBNO_Cue
root=Mixer_1
nodes [
 {id=Delay_0, children [Random_1], properties {delayMax=0.1, delayMin=0.1}}
 {id=Delay_1, children [Random_0], properties {delayMax=1.6, delayMin=1.6}}
 {id=Delay_2, children [Mixer_0], properties {delayMax=0.65, delayMin=0.65}}
 {id=Mixer_0, children [WavePlayer_0,Delay_0,Delay_1], properties {inputVolume [1.2,0.9,1.1]}}
 {id=Mixer_1, children [WavePlayer_1,Delay_2], properties {inputVolume [1,1]}}
 {id=Random_0, children [WavePlayer_6,WavePlayer_7,WavePlayer_8], properties {bShouldExcludeFromBranchCulling=true, weights [1,1,1]}}
 {id=Random_1, children [WavePlayer_2,WavePlayer_3,WavePlayer_5], properties {weights [1,1,1]}}
 {id=WavePlayer_0, properties {soundWaveAssetPtr=&/BackBling_Realm_Searching_Hover_01}}
 {id=WavePlayer_1, properties {soundWaveAssetPtr=&/BackBling_Realm_DamageReceived_01}}
 {id=WavePlayer_2, properties {soundWaveAssetPtr=&/BackBling_Realm_DBNO_Signal_01}}
 {id=WavePlayer_3, properties {soundWaveAssetPtr=&/BackBling_Realm_DBNO_Signal_02}}
 {id=WavePlayer_5, properties {soundWaveAssetPtr=&/BackBling_Realm_DBNO_Signal_04}}
 {id=WavePlayer_6, properties {soundWaveAssetPtr=&/BackBling_Realm_DBNO_Vox_01}}
 {id=WavePlayer_7, properties {soundWaveAssetPtr=&/BackBling_Realm_DBNO_Vox_02}}
 {id=WavePlayer_8, properties {soundWaveAssetPtr=&/BackBling_Realm_DBNO_Vox_03}}
]
properties {
 attenuationSettings=&/Fort_Realm_Close_Athena_Att
 audioPropertiesSheet=/BRCosmetics/Sounds/Settings/PropertySheets/Reactive/APS_Cosmetic_Reactive_Backbling
 maxDistance=1800.0001
 soundClassObject=/Game/Sounds/FortSoundClassesAndModes/Fort_Backblings
 volumeMultiplier=0.5
 concurrencyOverrides {bLimitToOwner=true, maxCount=1, resolutionRule=PreventNew}
 concurrencySet [&/BackBling_Realm]
}
